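# Post-copy animation Blueprint for the RedOasisGooseberry Icon character body, adding secondary motion and leg-driven correction to the jacket.
# AnimGraph copies the parent mesh pose, rotation-constrains the left front jacket bone to the thigh, then runs RigidBody simulation.
# EventGraph sets state-based gravity and lowers inertia scale while skydiving.
# During crouched movement, thigh swing angles, gated by north or south movement direction, weight additive offsets on both front jacket bones after simulation.
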AnimBlueprint /BRCosmetics/Characters/Player/Male/Medium/Bodies/M_MED_RedOasisGooseberry_Icon/Meshes/M_MED_RedOasisGooseberry_Icon_AnimBP
target_skeleton=&/M_MED_RedOasisGooseberry_Icon_Skeleton
graphs [
 AnimGraph {nodes [
   n1 = "Constraint - dyn_jacket_fr_l_MSTR (R)"!(
    ComponentPose=Local To Component(LocalPose=Copy Pose From Mesh() {bUseAttachedParent=true})
    bAlphaBoolEnabled:bool=True
    ConstraintWeights_0=1
   ) -> (Pose) {
    alphaInputType=Bool
    alphaBoolBlend {blendOption=Linear}
    alphaScaleBiasClamp {bInterpResult=true, interpSpeedDecreasing=2, interpSpeedIncreasing=2}
    boneToModify {boneName=dyn_jacket_fr_l_MSTR}
    constraintSetup [{transformType=Rotation, targetBone {boneName=thigh_l}}]
    constraintWeights [1]
   }
   n2 = Get ComponentScale() -> (ComponentScale:Vector)
   n3 = RigidBody!(
    ComponentPose=@n1
    Alpha
    OverrideWorldGravity:Vector=Get Override World Gravity() -> (Override World Gravity:Vector)
    ComponentLinearAccScale:Vector=@n2
    ComponentLinearVelScale:Vector=@n2
   ) -> (Pose) {
    bOverrideWorldGravity=true
    overridePhysicsAsset=&/M_MED_RedOasis_Gooseberry_Icon_Physics
    simulationSpace=BaseBoneSpace
    alphaBoolBlend {blendOption=Linear}
    baseBoneRef {boneName=pelvis}
    componentLinearAccScale=<0.4 0.4 0.4>
    componentLinearVelScale=<0.4 0.4 0.4>
    overrideWorldGravity {z=-2800}
   }
   n4 = "Transform (Modify) Bone - Bone: dyn_jacket_fr_l_01"(ComponentPose=@n3, Alpha=Get lNorth() -> (lNorth:double), Translation="-0.000,-0.607,4.409", Rotation="-34.445,2.031,-4.581", Scale="1,1,1") {
    rotationMode=BMM_Additive
    translationMode=BMM_Additive
    alphaBoolBlend {blendOption=Linear}
    alphaScaleBiasClamp {bInterpResult=true, interpSpeedDecreasing=50, interpSpeedIncreasing=50}
    boneToModify {boneName=dyn_jacket_fr_l_01}
    rotation=<rot -34.445 2.031 -4.581>
    translation {y=-0.607, z=4.409}
   }
   n5 = "Transform (Modify) Bone - Bone: dyn_jacket_fr_l_01"(ComponentPose=@n4, Alpha=Get lSouth() -> (lSouth:double), Translation="-0.000,-3.290,4.413", Rotation="-42.527,8.681,-9.510", Scale="1,1,1") {
    rotationMode=BMM_Additive
    translationMode=BMM_Additive
    alphaBoolBlend {blendOption=Linear}
    alphaScaleBiasClamp {bInterpResult=true, interpSpeedDecreasing=50, interpSpeedIncreasing=50}
    boneToModify {boneName=dyn_jacket_fr_l_01}
    rotation=<rot -42.527 8.681 -9.51>
    translation {y=-3.29, z=4.413}
   }
   n6 = "Transform (Modify) Bone - Bone: dyn_jacket_fr_r_01"(ComponentPose=@n5, Alpha=Get rNorth() -> (rNorth:double), Translation="1.647,-0.855,-0.000", Rotation="-0.000,-8.231,-19.421", Scale="1,1,1") {
    rotationMode=BMM_Additive
    translationMode=BMM_Additive
    alphaBoolBlend {blendOption=Linear}
    alphaScaleBiasClamp {bInterpResult=true, interpSpeedDecreasing=50, interpSpeedIncreasing=50}
    boneToModify {boneName=dyn_jacket_fr_r_01}
    rotation=<rot 0 -8.231 -19.421>
    translation=<1.647 -0.855>
   }
   n7 = "Transform (Modify) Bone - Bone: dyn_jacket_fr_r_01"(ComponentPose=@n6, Alpha=Get rSouth() -> (rSouth:double), Translation="0.000,0.000,2.782", Rotation="24.436,-8.785,-26.645", Scale="1,1,1") {
    rotationMode=BMM_Additive
    translationMode=BMM_Additive
    alphaBoolBlend {blendOption=Linear}
    alphaScaleBiasClamp {bInterpResult=true, interpSpeedDecreasing=50, interpSpeedIncreasing=50}
    boneToModify {boneName=dyn_jacket_fr_r_01}
    rotation=<rot 24.436 -8.785 -26.645>
    translation {x=0, z=2.7822}
   }
   Output Pose(Result=Component To Local(ComponentPose=@n7)) {layerGroup=None}
  ]}
 EventGraph {nodes [
   Event BlueprintUpdateAnimation()
   > n1 = Sequence!(execute) -> (then_0, then_1, then_2:Exec)
   > n2 = Gravity Override State Function DEPRECATED(CharacterPartAnimInstance=Self-Reference(), GameTimeInSeconds=GetGameTimeInSeconds(), GravityOverrideParams=Get Gravity Override Params())
   > Set Override World Gravity(Override World Gravity:Vector=@n2) -> (Output_Get:Vector)
   n3 = Get bIsMoving2D()
   > Set constr(constr:bool=AND Boolean(A=Get bIsCrouching(), B=NOT Boolean(A=@n3))) -> (Output_Get:bool)
   > n4 = Branch(Condition=Get bIsSkyDiving())
   > Set ComponentScale(ComponentScale:Vector="0.1,0.1,0.1") -> (Output_Get:Vector)
   Set ComponentScale(execute=@n4.else, ComponentScale="0.4,0.4,0.4")
   n5 = Branch(execute=@n1.then_1, Condition=AND Boolean(A=Get bIsCrouching(), B=Get bIsMoving2D()))
   n6 = GetOwnerSkeletalMeshComponent()
   > Set LlegCrouch(LlegCrouch:double=MapRangeClamped(Value=GetTwistAndSwingAngleOfDeltaRotationFromRefPose(self=@n6, BoneName=thigh_l).OutSwingAngle, InRangeA=45, InRangeB=100)) -> (Output_Get:double)
   > Set RlegCrouch(RlegCrouch:double=MapRangeClamped(Value=GetTwistAndSwingAngleOfDeltaRotationFromRefPose(self=@n6, BoneName=thigh_r).OutSwingAngle, InRangeA=150, InRangeB=75)) -> (Output_Get:double)
   Set LlegCrouch(execute=@n5.else, LlegCrouch=0)
   > Set RlegCrouch(RlegCrouch=0)
   n7 = Enum to String!(Enumerator:EFortCardinalDirection Enum=GetCardinalDirectionFromAngle!(
     self:Fort Anim Instance Object Reference
     Angle:float=Get PawnMovementDirectionAzimuth() -> (PawnMovementDirectionAzimuth:float)
     DeadZoneAngle:float=0
    ) -> (ReturnValue:EFortCardinalDirection Enum=North)) -> (ReturnValue:String)
   n8 = "Equal Exactly (String)"(A=@n7, B=North)
   n9 = Get LlegCrouch() -> (LlegCrouch:double)
   Set lNorth(execute=@n1.then_2, lNorth:double=float * float(A="To Float (Boolean)"(InBool=@n8), B=@n9) {is=title K2Node_PromotableOperator}) -> (Output_Get:double)
   n10 = "Equal Exactly (String)"(A=@n7, B=South)
   > Set lSouth(lSouth:double=float * float(A=@n9, B="To Float (Boolean)"(InBool=@n10))) -> (Output_Get:double)
   n11 = Get RlegCrouch() -> (RlegCrouch:double)
   > Set rNorth(rNorth:double=float * float(A="To Float (Boolean)"(InBool=@n8), B=@n11)) -> (Output_Get:double)
   > Set rSouth(rSouth:double=float * float(A=@n11, B="To Float (Boolean)"(InBool=@n10))) -> (Output_Get:double)
  ]}
]
variables [
 ComponentScale {type=Vector, default=<0.4 0.4 0.4>}
 constr {type=boolean}
 Gravity Override Params {
  type=GravityOverrideParamsStruct
  default {
   emoteOrMelee {gravityOverride=<0 0 -2800>}
   front_End {gravityOverride=<0 0 -2800>}
   onGround_Falling {gravityOverride=<0 0 -2800>}
   onGround_Moving {gravityOverride=<0 0 -2800>}
   onGround_Standing {gravityOverride=<0 0 -2800>}
   riding_Moving_Targeting {gravityOverride=<0 -200 -2800>}
   riding_Standing {gravityOverride=<0 -350 -2800>}
   skydive_DivingDown {jointName=root, gravityOverride=<0 0 0>, windAmplitude=<0 0 0>, windFrequency=<0 0 0>}
   skydive_FloatingDown {gravityOverride=<0 0 -300>, windAmplitude=<0 0 0>, windFrequency=<0 0 0>, windOffset=<0 0 0>}
   skydive_GoingUp {gravityOverride=<0 0 -300>, windAmplitude=<0 0 0>, windFrequency=<0 0 0>}
   skydive_Parachuting {gravityOverride=<0 -800 -2800>}
   skydive_Surfing {gravityOverride=<500 -300 -2800>}
   tactical_Sprint {gravityOverride=<0 0 -980>}
   ziplining {gravityOverride=<0 0 -2800>}
  }
 }
 LlegCrouch {type=number}
 lNorth {type=number}
 lSouth {type=number}
 Override World Gravity {type=Vector, default=<0 0 -2800>}
 RlegCrouch {type=number}
 rNorth {type=number}
 rSouth {type=number}
]
